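# BP_STPS_SharedWidgetState is an Editor Utility Blueprint that holds shared widget state, apparently as a property set for the workshop's Scriptable Tools.
# It stores two booleans, both on by default, named for whether the Readout and Shortcuts panels were open on tool exit.
# The event graph is empty, so it is a pure data holder with no read or write logic of its own.
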EditorUtilityBlueprint /Game/Blueprints/Blutilities/ToolsWrkshp/ScrTools/Shared/SharedPropertySets/BP_STPS_SharedWidgetState
graphs [EventGraph {links [], nodes []}]
variables [
 ReadoutOpenOnExit {type=boolean, default=true}
 ShortcutsOpenOnExit {type=boolean, default=true}
]
